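# Physics asset for the Cyber Scavenger male medium body cosmetic, set up for the coat and front-left collar dynamic bones.
# Pelvis, upper spine, legs and feet carry simple collision shapes, and a huge root capsule offset downward likely stands in for the ground.
# Small master boxes on the pelvis anchor tightly limited back and side coat capsule chains, while the collar hangs off spine_05.
# A free constraint links the two back chain tips.
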PhysicsAsset /BRCosmetics/Characters/Player/Male/Medium/Bodies/M_MED_Cyber_Scavenger/Meshes/M_MED_Cyber_Scavenger_Body_Physics
bodies {
 calf_l [{shapeType=Capsule, length=30.5982, radius=10.1518, center=<-23.6971 -2.7237 0.9718>, rotation=<rot -90 180 0>}]
 calf_r [{shapeType=Capsule, length=44.7859, radius=10.1518, center=<23.6971 2.7237 -0.9718>, rotation=<rot -87.667 94.1381 -94.1349>}]
 dyn_coat_back_l_01 [{shapeType=Capsule, length=10, radius=8, shapeName=dyn_coat_back_l_01_capsule, center=<0 -3 0>}]
 dyn_coat_back_l_02 [{shapeType=Capsule, length=24.9896, radius=8, shapeName=dyn_coat_back_l_02_capsule, center=<0 -1 0>}]
 dyn_coat_back_l_03 [{shapeType=Capsule, length=10, radius=8, shapeName=dyn_coat_back_l_03_capsule, center=<0 -3 0>}]
 dyn_coat_back_l_base [{shapeType=Capsule, length=10, radius=8, shapeName=dyn_coat_back_l_base_capsule, center=<0 -3 0>}]
 dyn_coat_back_r_01 [{shapeType=Capsule, length=10, radius=8, shapeName=dyn_coat_back_r_01_capsule, center=<0 -3 0>}]
 dyn_coat_back_r_02 [{shapeType=Capsule, length=10, radius=8, shapeName=dyn_coat_back_r_02_capsule, center=<0 -3 0>}]
 dyn_coat_back_r_03 [{shapeType=Capsule, length=10, radius=8, shapeName=dyn_coat_back_r_03_capsule, center=<0 -3 0>}]
 dyn_coat_back_r_base [{shapeType=Capsule, length=10, radius=8, shapeName=dyn_coat_back_r_base_capsule, center=<0 -3 0>}]
 dyn_coat_side_l_01 [
  {shapeType=Capsule, length=15, radius=6, shapeName=dyn_coat_side_l_01_capsule, center=<8 0 6.5>, rotation=<rot 100 0 0>}
  {shapeType=Capsule, length=15, radius=6, shapeName=dyn_coat_side_l_01_capsule, center=<8 -6.8224 -1.595>, rotation=<rot 0 0.0001 25.2892>}
 ]
 dyn_coat_side_l_base [{shapeType=Capsule, length=15, radius=6, shapeName=dyn_coat_side_l_base_capsule, center=<8 0 7>, rotation=<rot 100 0 0>}]
 dyn_coat_side_r_01 [
  {shapeType=Capsule, length=15, radius=6, shapeName=dyn_coat_side_r_01_capsule, center=<8 0 -7>, rotation=<rot 90 0 0>}
  {shapeType=Capsule, length=15, radius=6, shapeName=dyn_coat_side_r_01_capsule, center=<8 -8 -5>, rotation=<rot 90 0 0>}
 ]
 dyn_coat_side_r_base [{shapeType=Capsule, length=15, radius=6, shapeName=dyn_coat_side_r_base_capsule, center=<8 0 -7>, rotation=<rot 90 0 0>}]
 dyn_coller_front_l_01 [{shapeType=Box, extentX=11.51, extentY=2.01, extentZ=11.51, center=<2.7236 -0.1396 -3.8434>, rotation=<rot 7.053 -82.891 -29.5576>}]
 foot_l [{
   shapeType=Capsule
   length=15.5135
   radius=8.6108
   shapeName=foot_l_capsule
   center=<-3.8218 2.2438 -0.3773>
   rotation=<rot -1.1515 -172.1635 92.4269>
  }]
 foot_r [{
   shapeType=Capsule
   length=19.1179
   radius=9.4664
   shapeName=foot_r_capsule
   center=<4.1972 -4.9538 0.2613>
   rotation=<rot -1.1515 -172.1635 -87.5731>
  }]
 master_coat_back_l [{shapeType=Box, extentX=1.01, extentY=1.01, extentZ=1.01, shapeName=master_coat_back_l_box}]
 master_coat_back_r [{shapeType=Box, extentX=1.01, extentY=1.01, extentZ=1.01, shapeName=master_coat_back_r_box}]
 master_coat_side_l [{shapeType=Box, extentX=1.01, extentY=1.01, extentZ=1.01, shapeName=master_coat_side_l_box}]
 master_coat_side_r [{shapeType=Box, extentX=1.01, extentY=1.01, extentZ=1.01, shapeName=master_coat_side_r_box}]
 pelvis [{shapeType=Box, extentX=2.188, extentY=1.929, extentZ=1.623, center=<-0.032 -0.23 -0.5669>, rotation=<rot -2.2027 -8.8132 -50.9334>}]
 root [{shapeType=Capsule, length=0.505, radius=200, shapeName=root_capsule, center=<0 0 -192>, rotation=<rot 0 -90 0>}]
 spine_05 [{shapeType=Sphere, radius=14.2628}]
 thigh_l [{shapeType=Capsule, length=30, radius=10, shapeName=thigh_l_capsule, center=<-16.5171 0 0>, rotation=<rot 90 180 0>}]
 thigh_r [{shapeType=Capsule, length=30, radius=10, shapeName=thigh_r_capsule, center=<16.5171 0 0>, rotation=<rot 87.667 85.8616 85.8652>}]
}
constraints [
 {
  bone1Name=spine_05
  bone2Name=pelvis
  swing1LimitDegrees=45
  swing1Motion=Free
  swing2LimitDegrees=45
  swing2Motion=Free
  twistLimitDegrees=45
  twistMotion=Free
 }
 {bone1Name=dyn_coller_front_l_01, bone2Name=spine_05, swing1LimitDegrees=20, swing2LimitDegrees=10, twistLimitDegrees=5}
 {
  bone1Name=thigh_r
  bone2Name=pelvis
  swing1LimitDegrees=45
  swing1Motion=Free
  swing2LimitDegrees=45
  swing2Motion=Free
  twistLimitDegrees=45
  twistMotion=Free
 }
 {
  bone1Name=thigh_l
  bone2Name=pelvis
  swing1LimitDegrees=45
  swing1Motion=Free
  swing2LimitDegrees=45
  swing2Motion=Free
  twistLimitDegrees=45
  twistMotion=Free
 }
 {
  bone1Name=calf_l
  bone2Name=thigh_l
  swing1LimitDegrees=45
  swing1Motion=Free
  swing2LimitDegrees=45
  swing2Motion=Free
  twistLimitDegrees=45
  twistMotion=Free
 }
 {
  bone1Name=foot_l
  bone2Name=calf_l
  swing1LimitDegrees=45
  swing1Motion=Free
  swing2LimitDegrees=45
  swing2Motion=Free
  twistLimitDegrees=45
  twistMotion=Free
 }
 {
  bone1Name=calf_r
  bone2Name=thigh_r
  swing1LimitDegrees=45
  swing1Motion=Free
  swing2LimitDegrees=45
  swing2Motion=Free
  twistLimitDegrees=45
  twistMotion=Free
 }
 {
  bone1Name=foot_r
  bone2Name=calf_r
  swing1LimitDegrees=45
  swing1Motion=Free
  swing2LimitDegrees=45
  swing2Motion=Free
  twistLimitDegrees=45
  twistMotion=Free
 }
 {
  bone1Name=master_coat_side_l
  bone2Name=pelvis
  swing1LimitDegrees=45
  swing1Motion=Free
  swing2LimitDegrees=45
  swing2Motion=Free
  twistLimitDegrees=45
  twistMotion=Free
 }
 {
  bone1Name=master_coat_side_r
  bone2Name=pelvis
  swing1LimitDegrees=45
  swing1Motion=Free
  swing2LimitDegrees=45
  swing2Motion=Free
  twistLimitDegrees=45
  twistMotion=Free
 }
 {
  bone1Name=master_coat_back_l
  bone2Name=pelvis
  swing1LimitDegrees=45
  swing1Motion=Free
  swing2LimitDegrees=45
  swing2Motion=Free
  twistLimitDegrees=45
  twistMotion=Free
 }
 {bone1Name=dyn_coat_back_l_base, bone2Name=master_coat_back_l, swing1LimitDegrees=20, swing2LimitDegrees=5, twistLimitDegrees=1}
 {bone1Name=dyn_coat_back_l_01, bone2Name=dyn_coat_back_l_base, swing1LimitDegrees=20, swing2LimitDegrees=5, twistLimitDegrees=1}
 {bone1Name=dyn_coat_back_l_02, bone2Name=dyn_coat_back_l_01, swing1LimitDegrees=20, swing2LimitDegrees=5, twistLimitDegrees=1}
 {bone1Name=dyn_coat_back_l_03, bone2Name=dyn_coat_back_l_02, swing1LimitDegrees=20, swing2LimitDegrees=5, twistLimitDegrees=1}
 {
  bone1Name=master_coat_back_r
  bone2Name=pelvis
  swing1LimitDegrees=45
  swing1Motion=Free
  swing2LimitDegrees=45
  swing2Motion=Free
  twistLimitDegrees=45
  twistMotion=Free
 }
 {bone1Name=dyn_coat_back_r_base, bone2Name=master_coat_back_r, swing1LimitDegrees=20, swing2LimitDegrees=5, twistLimitDegrees=1}
 {bone1Name=dyn_coat_back_r_01, bone2Name=dyn_coat_back_r_base, swing1LimitDegrees=20, swing2LimitDegrees=5, twistLimitDegrees=1}
 {bone1Name=dyn_coat_back_r_02, bone2Name=dyn_coat_back_r_01, swing1LimitDegrees=20, swing2LimitDegrees=5, twistLimitDegrees=1}
 {bone1Name=dyn_coat_back_r_03, bone2Name=dyn_coat_back_r_02, swing1LimitDegrees=20, swing2LimitDegrees=5, twistLimitDegrees=1}
 {
  bone1Name=dyn_coat_back_r_03
  bone2Name=dyn_coat_back_l_03
  swing1LimitDegrees=45
  swing1Motion=Free
  swing2LimitDegrees=45
  swing2Motion=Free
  twistLimitDegrees=45
  twistMotion=Free
 }
 {bone1Name=dyn_coat_side_l_base, bone2Name=master_coat_side_l, swing1LimitDegrees=10, swing2LimitDegrees=10, twistLimitDegrees=0}
 {bone1Name=dyn_coat_side_l_01, bone2Name=dyn_coat_side_l_base, swing1LimitDegrees=10, swing2LimitDegrees=10, twistLimitDegrees=0}
 {bone1Name=dyn_coat_side_r_base, bone2Name=master_coat_side_r, swing1LimitDegrees=10, swing2LimitDegrees=10, twistLimitDegrees=0}
 {bone1Name=dyn_coat_side_r_01, bone2Name=dyn_coat_side_r_base, swing1LimitDegrees=10, swing2LimitDegrees=10, twistLimitDegrees=0}
 {
  bone1Name=pelvis
  bone2Name=root
  swing1LimitDegrees=45
  swing1Motion=Free
  swing2LimitDegrees=45
  swing2Motion=Free
  twistLimitDegrees=45
  twistMotion=Free
 }
]
