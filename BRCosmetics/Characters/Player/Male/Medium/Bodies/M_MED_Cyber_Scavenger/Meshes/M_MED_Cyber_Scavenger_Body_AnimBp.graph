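# Post-process style AnimBlueprint for the Cyber Scavenger male body, adding secondary motion to collar, belt loop, zipper and coat.
# It copies the attached parent's pose, then offsets collar bones by targeting state and head yaw relative to spine_05.
# Constraints, AnimDynamics and a pelvis-space RigidBody node feed stacked coat-base offsets.
# The event graph also sets slide, zipline, front-end and parameter-asset gravity values, mostly without visible readers in the AnimGraph.
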AnimBlueprint /BRCosmetics/Characters/Player/Male/Medium/Bodies/M_MED_Cyber_Scavenger/Meshes/M_MED_Cyber_Scavenger_Body_AnimBp
target_skeleton=&/M_MED_Cyber_Scavenger_Body_Skeleton
graphs [
 AnimGraph {nodes [
   n1 = Get ads_rotate() -> (ads_rotate:double)
   n2 = "Transform (Modify) Bone - Bone: dyn_coller_side_r_02"(
    ComponentPose=Local To Component(LocalPose=Copy Pose From Mesh() {bUseAttachedParent=true})
    Alpha=@n1
    Translation="-2,0,0"
    Rotation="15,-40,0"
    Scale="1,1,1"
   ) {
    rotationMode=BMM_Additive
    rotationSpace=BCS_BoneSpace
    translationMode=BMM_Additive
    translationSpace=BCS_BoneSpace
    alphaBoolBlend {blendOption=Linear}
    boneToModify {boneName=dyn_coller_side_r_02}
    rotation {pitch=15, yaw=-40}
    translation {x=-2}
   }
   n3 = "Transform (Modify) Bone - Bone: dyn_coller_front_r_01"(ComponentPose=@n2, Alpha=0, Translation="-3.5,0,0", Rotation="120,-25,-20", Scale="1,1,1") {
    rotationMode=BMM_Additive
    rotationSpace=BCS_BoneSpace
    translationMode=BMM_Additive
    translationSpace=BCS_BoneSpace
    alphaBoolBlend {blendOption=Linear}
    boneToModify {boneName=dyn_coller_front_r_01}
    rotation=<rot 120 -25 -20>
    translation {x=-3.5}
   }
   n4 = "Transform (Modify) Bone - Bone: dyn_coller_side_l_02"(ComponentPose=@n3, Alpha=@n1, Translation="3,0,-2", Rotation="20,15,0", Scale="1,1,1") {
    rotationMode=BMM_Additive
    rotationSpace=BCS_BoneSpace
    translationMode=BMM_Additive
    translationSpace=BCS_BoneSpace
    alphaBoolBlend {blendOption=Linear}
    boneToModify {boneName=dyn_coller_side_l_02}
    rotation {pitch=20, yaw=15}
    translation {x=3, z=-2}
   }
   n5 = "Transform (Modify) Bone - Bone: dyn_coller_front_l_01"(ComponentPose=@n4, Alpha=@n1, Translation="3,0,20", Rotation="0,0,0", Scale="1,1,1") {
    rotationMode=BMM_Additive
    rotationSpace=BCS_BoneSpace
    translationMode=BMM_Additive
    translationSpace=BCS_BoneSpace
    alphaBoolBlend {blendOption=Linear}
    boneToModify {boneName=dyn_coller_front_l_01}
    translation {x=3, z=20}
   }
   n6 = "Transform (Modify) Bone - Bone: dyn_coller_front_l_02"(ComponentPose=@n5, Alpha=@n1, Translation="2,0,2", Rotation="0,0,0", Scale="1,1,1") {
    rotationMode=BMM_Additive
    rotationSpace=BCS_BoneSpace
    translationMode=BMM_Additive
    translationSpace=BCS_BoneSpace
    alphaBoolBlend {blendOption=Linear}
    boneToModify {boneName=dyn_coller_front_l_02}
    translation {x=2, z=2}
   }
   n7 = Get head_rot_back() -> (head_rot_back:double)
   n8 = "Transform (Modify) Bone - Bone: dyn_coller_side_r_02"(ComponentPose=@n6, Alpha=@n7, Translation="2,0,0", Rotation="-20,0,0", Scale="1,1,1") {
    rotationMode=BMM_Additive
    rotationSpace=BCS_BoneSpace
    alphaBoolBlend {blendOption=Linear}
    boneToModify {boneName=dyn_coller_side_r_02}
    rotation {pitch=-20}
    translation {x=2}
   }
   n9 = "Transform (Modify) Bone - Bone: dyn_coller_back_01"(ComponentPose=@n8, Alpha=@n7, Translation="0,0,0", Rotation="-70,0,0", Scale="1,1,1") {
    rotationMode=BMM_Additive
    rotationSpace=BCS_BoneSpace
    alphaBoolBlend {blendOption=Linear}
    boneToModify {boneName=dyn_coller_back_01}
    rotation {pitch=-70}
   }
   n10 = "Transform (Modify) Bone - Bone: dyn_coller_front_r_01"(ComponentPose=@n9, Alpha=SelectFloat(A=@n7, B=0, bPickA=Get bIsTargeting()), Translation="-5,7,9", Rotation="120,10,0", Scale="1,1,1") {
    rotationMode=BMM_Additive
    rotationSpace=BCS_BoneSpace
    translationMode=BMM_Replace
    translationSpace=BCS_BoneSpace
    alphaBoolBlend {blendOption=Linear}
    alphaScaleBiasClamp {bMapRange=true, inRange {max=0.45}}
    boneToModify {boneName=dyn_coller_front_r_01}
    rotation {pitch=120, yaw=10}
    translation=<-5 7 9>
   }
   n11 = "Constraint - dyn_back_beltloop (PPPPPPPP)"!(
    ComponentPose=@n10
    Alpha=1
    ConstraintWeights_0=0.0526
    ConstraintWeights_1:float=0.1629
    ConstraintWeights_2:float=0.0785
    ConstraintWeights_3:float=0.345
    ConstraintWeights_4:float=0.0846
    ConstraintWeights_5:float=0.0035
    ConstraintWeights_6:float=0.0503
    ConstraintWeights_7:float=0.2228
   ) -> (Pose) {
    alphaBoolBlend {blendOption=Linear}
    boneToModify {boneName=dyn_back_beltloop}
    constraintSetup [
     {targetBone {boneName=clavicle_l}}
     {targetBone {boneName=spine_01}}
     {targetBone {boneName=spine_02}}
     {targetBone {boneName=spine_03}}
     {targetBone {boneName=spine_04}}
     {targetBone {boneName=spine_05}}
     {targetBone {boneName=clavicle_r}}
     {targetBone {boneName=pelvis}}
    ]
    constraintWeights [1,1,1,1,1,1,1,1]
   }
   n12 = Anim Dynamics - Bone: dyn_back_beltloop!(SimulationSpace:AnimPhysSimSpaceType Enum=Component, ComponentPose=@n11, Alpha=1, ExternalForce="0,0,0") -> (Pose) {
    bGravityOverrideInSimSpace=true
    bUseSphericalLimits=true
    alphaBoolBlend {blendOption=Linear}
    boundBone {boneName=dyn_back_beltloop}
    componentAppliedLinearAccClamp=<7000 7000 7000>
    componentLinearAccScale=<0.5 0.5 0.5>
    componentLinearVelScale=<0.5 0.5 0.5>
    physicsBodyDefinitions [{
      sphereCollisionRadius=0
      boundBone {boneName=dyn_back_beltloop}
      boxExtents=<10 3 10>
      constraintSetup {angularLimitsMax=<10 0 5>, angularLimitsMin=<-10 0 -15>, angularTarget=<0 0 0>}
      localJointOffset=<3 0 0>
     }]
    planarLimits [{drivingBone {boneName=spine_03}, planeTransform {location=<0 -11 0>, rotation=<rot 90 89.9999 0>}}]
    relativeSpaceBone {boneName=pelvis}
    sphericalLimits [{limitRadius=10.4, drivingBone {boneName=dyn_back_beltloop}, sphereLocalOffset=<2 10.5 0>}]
   }
   n13 = "Constraint - dyn_zipper (PPPPPPPP)"!(
    ComponentPose=@n12
    Alpha=1
    ConstraintWeights_0=0.01
    ConstraintWeights_1:float=0.0061
    ConstraintWeights_2:float=0.0021
    ConstraintWeights_3:float=0.001
    ConstraintWeights_4:float=0
    ConstraintWeights_5:float=0
    ConstraintWeights_6:float=0.586
    ConstraintWeights_7:float=0.3948
   ) -> (Pose) {
    alphaBoolBlend {blendOption=Linear}
    boneToModify {boneName=dyn_zipper}
    constraintSetup [
     {targetBone {boneName=thigh_twist_01_r}}
     {targetBone {boneName=spine_01}}
     {targetBone {boneName=spine_02}}
     {targetBone {boneName=spine_03}}
     {targetBone {boneName=spine_04}}
     {targetBone {boneName=thigh_l}}
     {targetBone {boneName=pelvis}}
     {targetBone {boneName=thigh_twist_01_l}}
    ]
    constraintWeights [1,1,1,1,1,1,1,1]
   }
   n14 = Anim Dynamics - Bone: dyn_zipper!(SimulationSpace:AnimPhysSimSpaceType Enum=World, ComponentPose=@n13, Alpha=1, GravityOverride:Vector="0,0,0", ExternalForce) -> (Pose) {
    angularDampingOverride=0.9999
    angularSpringConstant=15000
    bAngularSpring=true
    bGravityOverrideInSimSpace=true
    bOverrideAngularDamping=true
    bUseGravityOverride=true
    bUseSphericalLimits=true
    simulationSpace=World
    alphaBoolBlend {blendOption=Linear}
    boundBone {boneName=dyn_zipper}
    componentAppliedLinearAccClamp=<4000 4000 4000>
    componentLinearAccScale=<0.25 0.25 0.25>
    componentLinearVelScale=<0.25 0.25 0.25>
    physicsBodyDefinitions [{
      sphereCollisionRadius=0
      boundBone {boneName=dyn_zipper}
      boxExtents=<10 10 10>
      constraintSetup {angularLimitsMax=<0 10 5>, angularLimitsMin=<0 -30 -5>, angularTarget=<0 0 0>}
      localJointOffset=<3 0 0>
     }]
    relativeSpaceBone {boneName=pelvis}
    sphericalLimits [{limitRadius=3, drivingBone {boneName=dyn_zipper}, sphereLocalOffset=<2 0 -3.25>}]
   }
   n15 = RigidBody!(
    ComponentPose=@n14
    Alpha
    OverrideWorldGravity:Vector="0,0,-980"
    ExternalForce
    SimSpaceSettings:Sim Space Settings Structure=(WorldAlpha=1,VelocityScaleZ=1,DampingAlpha=1,MaxLinearVelocity=10000,MaxAngularVelocity=10000,MaxLinearAcceleration=10000,MaxAngularAcceleration=10000,ExternalLinearDragV=(X=0,Y=0,Z=0),ExternalLinearVelocity=(X=0,Y=0,Z=0),ExternalAngularVelocity=(X=0,Y=0,Z=0))
   ) -> (Pose) {
    bClampLinearTranslationLimitToRefPose=true
    bOverrideWorldGravity=true
    overridePhysicsAsset=&/M_MED_Cyber_Scavenger_Body_Physics
    simulationSpace=BaseBoneSpace
    alphaBoolBlend {blendOption=Linear}
    baseBoneRef {boneName=pelvis}
    componentAppliedLinearAccClamp=<2000 2000 2000>
    overrideWorldGravity {z=-980}
    simSpaceSettings {worldAlpha=1}
   }
   n16 = Anim Dynamics - Bone: dyn_coat_back_r_base!(SimulationSpace:AnimPhysSimSpaceType Enum, ComponentPose=@n15, Alpha=1, GravityOverride:Vector="0,0,-980", ExternalForce="0,0,0") -> (Pose) {
    angularDampingOverride=1
    angularSpringConstant=10000
    bAngularSpring=true
    bOverrideAngularDamping=true
    bUseGravityOverride=true
    gravityScale=0.35
    simulationSpace=RootRelative
    boundBone {boneName=dyn_coat_back_r_base}
    componentAppliedLinearAccClamp=<2000 2000 2000>
    componentLinearAccScale=<0.5 0.5 0.5>
    componentLinearVelScale=<0.5 1 0.5>
    gravityOverride {z=-980}
    physicsBodyDefinitions [{
      boundBone {boneName=dyn_coat_back_r_base}
      boxExtents=<20 30 30>
      constraintSetup {angularLimitsMax=<0 5 0>, angularLimitsMin=<0 -5 -20>, angularTarget=<1 0 0>}
      localJointOffset=<4 0 0>
     }]
   }
   n17 = Anim Dynamics - Bone: dyn_coat_back_r_01!(SimulationSpace:AnimPhysSimSpaceType Enum, ComponentPose=@n16, Alpha=1, GravityOverride:Vector="0,0,-980", ExternalForce="0,0,0") -> (Pose) {
    angularDampingOverride=1
    angularSpringConstant=10000
    bAngularSpring=true
    bOverrideAngularDamping=true
    bUseGravityOverride=true
    gravityScale=0.35
    simulationSpace=RootRelative
    boundBone {boneName=dyn_coat_back_r_01}
    componentAppliedLinearAccClamp=<2000 2000 2000>
    componentLinearAccScale=<0.5 0.5 0.5>
    componentLinearVelScale=<0.5 1 0.5>
    gravityOverride {z=-980}
    physicsBodyDefinitions [{
      boundBone {boneName=dyn_coat_back_r_01}
      boxExtents=<20 30 30>
      constraintSetup {angularLimitsMax=<0 5 0>, angularLimitsMin=<0 -5 -20>, angularTarget=<1 0 0>}
      localJointOffset=<4 0 0>
     }]
   }
   n18 = Anim Dynamics - Bone: dyn_coat_back_r_02!(SimulationSpace:AnimPhysSimSpaceType Enum, ComponentPose=@n17, Alpha=1, GravityOverride:Vector="0,0,-980", ExternalForce="0,0,0") -> (Pose) {
    angularDampingOverride=1
    angularSpringConstant=10000
    bAngularSpring=true
    bOverrideAngularDamping=true
    bUseGravityOverride=true
    gravityScale=0.35
    simulationSpace=RootRelative
    boundBone {boneName=dyn_coat_back_r_02}
    componentAppliedLinearAccClamp=<2000 2000 2000>
    componentLinearAccScale=<0.5 0.5 0.5>
    componentLinearVelScale=<0.5 1 0.5>
    gravityOverride {z=-980}
    physicsBodyDefinitions [{
      boundBone {boneName=dyn_coat_back_r_02}
      boxExtents=<20 30 30>
      constraintSetup {angularLimitsMax=<0 5 0>, angularLimitsMin=<0 -5 -20>, angularTarget=<1 0 0>}
      localJointOffset=<4 0 0>
     }]
   }
   n19 = Anim Dynamics - Bone: dyn_coat_back_r_03!(SimulationSpace:AnimPhysSimSpaceType Enum, ComponentPose=@n18, Alpha=1, GravityOverride:Vector="0,0,-980", ExternalForce="0,0,0") -> (Pose) {
    angularDampingOverride=1
    angularSpringConstant=10000
    bAngularSpring=true
    bOverrideAngularDamping=true
    bUseGravityOverride=true
    gravityScale=0.35
    simulationSpace=RootRelative
    boundBone {boneName=dyn_coat_back_r_03}
    componentAppliedLinearAccClamp=<2000 2000 2000>
    componentLinearAccScale=<0.5 0.5 0.5>
    componentLinearVelScale=<0.5 1 0.5>
    gravityOverride {z=-980}
    physicsBodyDefinitions [{
      boundBone {boneName=dyn_coat_back_r_03}
      boxExtents=<20 30 30>
      constraintSetup {angularLimitsMax=<0 5 0>, angularLimitsMin=<0 -5 -20>, angularTarget=<1 0 0>}
      localJointOffset=<4 0 0>
     }]
   }
   n20 = Anim Dynamics - Bone: dyn_coat_back_l_base!(SimulationSpace:AnimPhysSimSpaceType Enum, ComponentPose=@n19, Alpha=1, GravityOverride:Vector="0,0,-980", ExternalForce="0,0,0") -> (Pose) {
    angularDampingOverride=1
    angularSpringConstant=10000
    bAngularSpring=true
    bOverrideAngularDamping=true
    bUseGravityOverride=true
    gravityScale=0.35
    simulationSpace=RootRelative
    boundBone {boneName=dyn_coat_back_l_base}
    componentAppliedLinearAccClamp=<2000 2000 2000>
    componentLinearAccScale=<0.5 0.5 0.5>
    componentLinearVelScale=<0.5 1 0.5>
    gravityOverride {z=-980}
    physicsBodyDefinitions [{
      boundBone {boneName=dyn_coat_back_l_base}
      boxExtents=<20 30 30>
      constraintSetup {angularLimitsMax=<0 5 0>, angularLimitsMin=<0 -5 -20>, angularTarget=<1 0 0>}
      localJointOffset=<4 0 0>
     }]
   }
   n21 = Anim Dynamics - Bone: dyn_coat_back_l_01!(SimulationSpace:AnimPhysSimSpaceType Enum, ComponentPose=@n20, Alpha=1, GravityOverride:Vector="0,0,-980", ExternalForce="0,0,0") -> (Pose) {
    angularDampingOverride=1
    angularSpringConstant=10000
    bAngularSpring=true
    bOverrideAngularDamping=true
    bUseGravityOverride=true
    gravityScale=0.35
    simulationSpace=RootRelative
    boundBone {boneName=dyn_coat_back_l_01}
    componentAppliedLinearAccClamp=<2000 2000 2000>
    componentLinearAccScale=<0.5 0.5 0.5>
    componentLinearVelScale=<0.5 1 0.5>
    gravityOverride {z=-980}
    physicsBodyDefinitions [{
      boundBone {boneName=dyn_coat_back_l_01}
      boxExtents=<20 30 30>
      constraintSetup {angularLimitsMax=<0 5 0>, angularLimitsMin=<0 -5 -20>, angularTarget=<1 0 0>}
      localJointOffset=<4 0 0>
     }]
   }
   n22 = Anim Dynamics - Bone: dyn_coat_back_l_02!(SimulationSpace:AnimPhysSimSpaceType Enum, ComponentPose=@n21, Alpha=1, GravityOverride:Vector="0,0,-980", ExternalForce="0,0,0") -> (Pose) {
    angularDampingOverride=1
    angularSpringConstant=10000
    bAngularSpring=true
    bOverrideAngularDamping=true
    bUseGravityOverride=true
    gravityScale=0.35
    simulationSpace=RootRelative
    boundBone {boneName=dyn_coat_back_l_02}
    componentAppliedLinearAccClamp=<2000 2000 2000>
    componentLinearAccScale=<0.5 0.5 0.5>
    componentLinearVelScale=<0.5 1 0.5>
    gravityOverride {z=-980}
    physicsBodyDefinitions [{
      boundBone {boneName=dyn_coat_back_l_02}
      boxExtents=<20 30 30>
      constraintSetup {angularLimitsMax=<0 5 0>, angularLimitsMin=<0 -5 -20>, angularTarget=<1 0 0>}
      localJointOffset=<4 0 0>
     }]
   }
   n23 = Anim Dynamics - Bone: dyn_coat_back_l_03!(SimulationSpace:AnimPhysSimSpaceType Enum, ComponentPose=@n22, Alpha=1, GravityOverride:Vector="0,0,-980", ExternalForce="0,0,0") -> (Pose) {
    angularDampingOverride=1
    angularSpringConstant=10000
    bAngularSpring=true
    bOverrideAngularDamping=true
    bUseGravityOverride=true
    gravityScale=0.35
    simulationSpace=RootRelative
    boundBone {boneName=dyn_coat_back_l_03}
    componentAppliedLinearAccClamp=<2000 2000 2000>
    componentLinearAccScale=<0.5 0.5 0.5>
    componentLinearVelScale=<0.5 1 0.5>
    gravityOverride {z=-980}
    physicsBodyDefinitions [{
      boundBone {boneName=dyn_coat_back_l_03}
      boxExtents=<20 30 30>
      constraintSetup {angularLimitsMax=<0 5 0>, angularLimitsMin=<0 -5 -20>, angularTarget=<1 0 0>}
      localJointOffset=<4 0 0>
     }]
   }
   n24 = "Transform (Modify) Bone - Bone: dyn_coat_side_l_base"(ComponentPose=@n23, Alpha=1, Translation="-10,0,0", Rotation="0,-80,0", Scale="1,1,1") {
    rotationMode=BMM_Additive
    rotationSpace=BCS_BoneSpace
    translationMode=BMM_Additive
    translationSpace=BCS_BoneSpace
    boneToModify {boneName=dyn_coat_side_l_base}
    rotation {yaw=-80}
    translation {x=-10}
   }
   n25 = "Transform (Modify) Bone - Bone: dyn_coat_back_l_base"(ComponentPose=@n24, Alpha=1, Translation="-10,0,0", Rotation="0,-80,0", Scale="1,1,1") {
    rotationMode=BMM_Additive
    rotationSpace=BCS_BoneSpace
    translationMode=BMM_Additive
    translationSpace=BCS_BoneSpace
    boneToModify {boneName=dyn_coat_back_l_base}
    rotation {yaw=-80}
    translation {x=-10}
   }
   n26 = "Transform (Modify) Bone - Bone: dyn_coat_back_r_base"(ComponentPose=@n25, Alpha=1, Translation="-15,0,0", Rotation="0,-80,0", Scale="1,1,1") {
    rotationMode=BMM_Additive
    rotationSpace=BCS_BoneSpace
    translationMode=BMM_Additive
    translationSpace=BCS_BoneSpace
    boneToModify {boneName=dyn_coat_back_r_base}
    rotation {yaw=-80}
    translation {x=-15}
   }
   n27 = "Transform (Modify) Bone - Bone: dyn_coat_side_r_base"(ComponentPose=@n26, Alpha=1, Translation="-15,0,0", Rotation="-20,-80,0", Scale="1,1,1") {
    rotationMode=BMM_Additive
    rotationSpace=BCS_BoneSpace
    translationMode=BMM_Additive
    translationSpace=BCS_BoneSpace
    boneToModify {boneName=dyn_coat_side_r_base}
    rotation {pitch=-20, yaw=-80}
    translation {x=-15}
   }
   n28 = "Transform (Modify) Bone - Bone: dyn_coat_side_l_base"(ComponentPose=@n27, Alpha=1, Translation="-10,0,0", Rotation="0,-120,0", Scale="1,1,1") {
    rotationMode=BMM_Additive
    rotationSpace=BCS_BoneSpace
    translationMode=BMM_Additive
    translationSpace=BCS_BoneSpace
    boneToModify {boneName=dyn_coat_side_l_base}
    rotation {yaw=-120}
    translation {x=-10}
   }
   n29 = "Transform (Modify) Bone - Bone: dyn_coat_back_l_base"(ComponentPose=@n28, Alpha=1, Translation="-10,0,0", Rotation="0,-120,0", Scale="1,1,1") {
    rotationMode=BMM_Additive
    rotationSpace=BCS_BoneSpace
    translationMode=BMM_Additive
    translationSpace=BCS_BoneSpace
    boneToModify {boneName=dyn_coat_back_l_base}
    rotation {yaw=-120}
    translation {x=-10}
   }
   n30 = "Transform (Modify) Bone - Bone: dyn_coat_back_r_base"(ComponentPose=@n29, Alpha=1, Translation="-5,0,0", Rotation="0,-120,0", Scale="1,1,1") {
    rotationMode=BMM_Additive
    rotationSpace=BCS_BoneSpace
    translationMode=BMM_Additive
    translationSpace=BCS_BoneSpace
    boneToModify {boneName=dyn_coat_back_r_base}
    rotation {yaw=-120}
    translation {x=-5}
   }
   n31 = "Transform (Modify) Bone - Bone: dyn_coat_side_r_base"(ComponentPose=@n30, Alpha=1, Translation="-10,0,0", Rotation="-20,-100,0", Scale="1,1,1") {
    rotationMode=BMM_Additive
    rotationSpace=BCS_BoneSpace
    translationMode=BMM_Additive
    translationSpace=BCS_BoneSpace
    boneToModify {boneName=dyn_coat_side_r_base}
    rotation {pitch=-20, yaw=-100}
    translation {x=-10}
   }
   n32 = "Transform (Modify) Bone - Bone: dyn_coat_back_r_base"!(ComponentPose=@n31, bAlphaBoolEnabled:bool=True, Translation="-1,0,0", Rotation="0,-5,0", Scale="1,1,1") -> (Pose) {
    alphaInputType=Bool
    rotationMode=BMM_Additive
    rotationSpace=BCS_BoneSpace
    translationMode=BMM_Additive
    translationSpace=BCS_BoneSpace
    boneToModify {boneName=dyn_coat_back_r_base}
    rotation {yaw=-5}
    translation {x=-1}
   }
   n33 = "Transform (Modify) Bone - Bone: dyn_coat_side_r_base"!(ComponentPose=@n32, bAlphaBoolEnabled:bool=True, Translation="-1,0,0", Rotation="0,-5,0", Scale="1,1,1") -> (Pose) {
    alphaInputType=Bool
    rotationMode=BMM_Additive
    rotationSpace=BCS_BoneSpace
    translationMode=BMM_Additive
    translationSpace=BCS_BoneSpace
    boneToModify {boneName=dyn_coat_side_r_base}
    rotation {yaw=-5}
    translation {x=-1}
   }
   n34 = "Transform (Modify) Bone - Bone: dyn_coat_back_l_base"!(ComponentPose=@n33, bAlphaBoolEnabled:bool=True, Translation="-1,0,0", Rotation="0,-5,0", Scale="1,1,1") -> (Pose) {
    alphaInputType=Bool
    rotationMode=BMM_Additive
    rotationSpace=BCS_BoneSpace
    translationMode=BMM_Additive
    translationSpace=BCS_BoneSpace
    boneToModify {boneName=dyn_coat_back_l_base}
    rotation {yaw=-5}
    translation {x=-1}
   }
   n35 = "Transform (Modify) Bone - Bone: dyn_coat_side_l_base"!(ComponentPose=@n34, bAlphaBoolEnabled:bool=True, Translation="-1,0,0", Rotation="0,-5,0", Scale="1,1,1") -> (Pose) {
    alphaInputType=Bool
    rotationMode=BMM_Additive
    rotationSpace=BCS_BoneSpace
    translationMode=BMM_Additive
    translationSpace=BCS_BoneSpace
    boneToModify {boneName=dyn_coat_side_l_base}
    rotation {yaw=-5}
    translation {x=-1}
   }
   Output Pose(Result=Component To Local(ComponentPose=@n35)) {layerGroup=None}
  ]}
 EventGraph {nodes [
   Event BlueprintUpdateAnimation()
   > n1 = Sequence!(execute) -> (then_0, then_1, then_2:Exec)
   n2 = BreakTransform(InTransform=GetDeltaTransformFromRefPose(self=GetOwnerSkeletalMeshComponent(), BoneName=head, BaseName=spine_05))
   > Set head_rot_back(head_rot_back:double=MapRangeClamped(Value=BreakRotator(InRot=@n2.Rotation).Yaw, InRangeA=0, InRangeB=-55)) -> (Output_Get:double)
   n3 = Get bIsTargeting()
   > Set ads_rotate(ads_rotate:double="To Float (Boolean)"(InBool=@n3)) -> (Output_Get:double)
   n4 = Get bIsKneeSliding()
   n5 = Get bIsBaseballSliding()
   Set coat_alpha(execute=@n1.then_1, coat_alpha:double=SelectFloat(A=0, B=1, bPickA=OR Boolean(A=@n4, B=@n5))) -> (Output_Get:double)
   n6 = "Gravity Override State Function (Parameter Asset) (Nativized)"(
    execute=@n1.then_2
    CharacterPartAnimInstance=Self-Reference()
    GameTimeInSeconds=GetGameTimeInSeconds()
    Parameters=&/M_MED_Cyber_Scavenger_Parameters
   )
   > Set GravityOverride(GravityOverride=@n6)
   > Set RBAN Settings Out(RBAN Settings Out=@n6.OutRigidBodyAnimNodeParameters)
   > n7 = Branch(Condition=Get bIsInFrontEnd())
   > "Set Sim Space "("Sim Space ":AnimPhysSimSpaceType Enum=World) -> (Output_Get:AnimPhysSimSpaceType Enum)
   > Set CoatDynAlpha(CoatDynAlpha:double=0.3) -> (Output_Get:double)
   "Set Sim Space "(execute=@n7.else)
   > Set CoatDynAlpha(CoatDynAlpha=1)
   n8 = Event BlueprintProcessAnimInputEvent()
   > n9 = Branch(Condition=IsNamed(self=@n8.AnimInputEvent, InEventName=aie.Player.Zipline.Entered))
   > Set OnZipline(OnZipline:bool=1) -> (Output_Get:bool)
   Branch(execute=@n9.else, Condition=IsNamed(self=@n8.AnimInputEvent, InEventName=aie.Player.Zipline.Exited))
   > Set OnZipline(OnZipline=false)
   Event OnBeginKneeSliding() {is=title K2Node_Event}
   > Set isKneeSlidling float(isKneeSlidling float:double=1) -> (Output_Get:double)
   Event OnEndKneeSliding() {is=title K2Node_Event}
   > Set isKneeSlidling float(isKneeSlidling float=0)
   Event OnBeginBaseballSliding() {is=title K2Node_Event}
   > Set isBaseballSlidling float(isBaseballSlidling float:double=1) -> (Output_Get:double)
   Event OnEndBaseballSliding() {is=title K2Node_Event}
   > Set isBaseballSlidling float(isBaseballSlidling float=0)
   TryGetPawnOwner()
   "To Float (Boolean)"(InBool=NOT Boolean(A=@n3))
   SelectFloat(
    A=0.3
    B=1
    bPickA=OR Boolean!(A=Get bIsSkyDiving(), B=Get bIsParachuteOpened(), C:bool=Get bIsSkydiveFloating(), D:bool=Get bIsSkydiveDiveMode()) -> (ReturnValue)
   )
  ]}
]
variables [
 ads_rotate {type=number}
 coat_alpha {type=number}
 CoatDynAlpha {type=number}
 GravityOverride {type=Vector}
 head_rot_back {type=number}
 isBaseballSlidling float {type=number}
 isKneeSlidling float {type=number}
 OnZipline {type=boolean}
 RBAN Settings Out {type=FortRigidBodyAnimNodeParameters}
 "Sim Space " {type=AnimPhysSimSpaceType}
]
